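# Third-person fire sound cue for the Figment pistol-caliber assault rifle, layering close, medium and distant gunshots.
# The close layer randomly picks one of two waves, while medium and distant play single waves, with distant turned down.
# Each layer has its own attenuation and sound class override, with medium sharing the Distant class.
# Close and medium are mixed and modulated before joining distant, under local and global gunfire concurrency sets.
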SoundCue /Figment_Assault_PistolCalibur_AR/Audio/SC_Assault_PistolCaliber_3P_Figment_Cue
root=Mixer_2
nodes [
 {
  id=Attenuation_3
  children [Random_2]
  properties {attenuationSettings=/Figment_Assault_PistolCalibur_AR/Audio/Attenuations/ATT_Projectile_Assault_PistolCaliber_Close}
 }
 {
  id=Attenuation_4
  children [Modulator_2]
  properties {attenuationSettings=/Figment_Assault_PistolCalibur_AR/Audio/Attenuations/ATT_Projectile_Assault_PistolCaliber_Distant}
 }
 {
  id=Attenuation_5
  children [Modulator_4]
  properties {attenuationSettings=/Figment_Assault_PistolCalibur_AR/Audio/Attenuations/ATT_Projectile_Assault_PistolCaliber_Medium}
 }
 {id=Mixer_2, children [Modulator_3,SoundClass_4], properties {inputVolume [1,1]}}
 {id=Mixer_3, children [SoundClass_3,SoundClass_5], properties {inputVolume [0.6,1]}}
 {id=Modulator_2, children [WavePlayer_5], properties {pitchMin=1, volumeMax=0.3, volumeMin=0.3}}
 {id=Modulator_3, children [Mixer_3]}
 {id=Modulator_4, children [WavePlayer_2], properties {pitchMin=1, volumeMax=1, volumeMin=1}}
 {id=Random_2, children [WavePlayer_3,WavePlayer_4], properties {weights [1,1]}}
 {
  id=SoundClass_3
  children [Attenuation_3]
  properties {soundClassOverride=&Fort_WeaponRanged_Fire_3P_PSM}
 }
 {
  id=SoundClass_4
  children [Attenuation_4]
  properties {soundClassOverride=&Fort_WeaponRanged_Fire_Distant}
 }
 {
  id=SoundClass_5
  children [Attenuation_5]
  properties {soundClassOverride=&Fort_WeaponRanged_Fire_Distant}
 }
 {id=WavePlayer_2, properties {soundWaveAssetPtr=&/JW_AR_Medium_Fire_01}}
 {id=WavePlayer_3, properties {soundWaveAssetPtr=/Game/Sounds/Fort_Weapons/AutoRifle/Assault_PistolCaliber/JW_AR_Close_Fire_02}}
 {id=WavePlayer_4, properties {soundWaveAssetPtr=/Game/Sounds/Fort_Weapons/AutoRifle/Assault_PistolCaliber/JW_AR_Close_Fire_03}}
 {id=WavePlayer_5, properties {soundWaveAssetPtr=&/JW_AR_Distant_Fire_01}}
]
properties {
 bDebug=true
 maxDistance=18000
 soundClassObject=&Fort_WeaponRanged_Fire_3P_PSM
 volumeMultiplier=1.1
 concurrencyOverrides {bLimitToOwner=true, maxCount=3, resolutionRule=StopOldest}
 concurrencySet [&Weapon_Gunfire_Global_Conc, &Weapon_Gunfire_Local_Conc]
}
