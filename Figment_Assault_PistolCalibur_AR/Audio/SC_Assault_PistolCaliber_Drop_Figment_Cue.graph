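# Sound cue for the drop sound of the Figment assault pistol-caliber AR weapon.
# A single wave player plays a shared Fortnite weapon drop wave at reduced volume, routed through the loot collect sound class.
# Weapon drop attenuation handles spatial falloff, and a shared stop-oldest weapon drop concurrency set limits overlapping drops.
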SoundCue /Figment_Assault_PistolCalibur_AR/Audio/SC_Assault_PistolCaliber_Drop_Figment_Cue
root=WavePlayer_5
nodes [{id=WavePlayer_5, properties {soundWaveAssetPtr=/Game/Sounds/Fort_GamePlay_Sounds/Loot/fort_weapon_drop_02_mix}}]
properties {
 attenuationSettings=&Fort_Weapon_Drop_Att
 maxDistance=3000
 soundClassObject=&Fort_Loot_Collect
 volumeMultiplier=0.5
 concurrencyOverrides {resolutionRule=PreventNew}
 concurrencySet [&Weapon_Drop_StopOldest]
}
