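# Sound cue for the empty-magazine sound of a Figment pistol-caliber assault rifle, classed as weapon foley.
# A mixer layers a pitched-up assault rifle reload element with a negative UI gameplay sound, each at reduced volume.
# Overridden concurrency allows two simultaneous instances and stops the oldest.
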SoundCue /Figment_Assault_PistolCalibur_AR/Audio/SC_Assault_PistolCaliber_Empty_Cue
root=Mixer_5
nodes [
 {id=Mixer_5, children [Modulator_4, WavePlayer_17], properties {inputVolume [0.5,0.4]}}
 {id=Modulator_4, children [WavePlayer_21], properties {pitchMax=1.3, pitchMin=1.3}}
 {id=WavePlayer_17, properties {soundWaveAssetPtr=/Game/Sounds/Fort_UI_Sounds/Gameplay/fort_ui_new_neg_01}}
 {id=WavePlayer_21, properties {soundWaveAssetPtr=/Game/Sounds/Fort_Weapons/AutoRifle/Reload_Elements/assault_end_sdk_01}}
]
properties {
 bOverrideConcurrency=true
 maxDistance=2097152
 soundClassObject=&Fort_WeaponFoley
 volumeMultiplier=1
 attenuationOverrides {falloffDistance=1600}
 concurrencyOverrides {maxCount=2, resolutionRule=StopOldest}
}
