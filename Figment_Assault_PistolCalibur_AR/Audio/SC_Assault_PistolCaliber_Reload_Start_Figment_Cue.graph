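# Sound cue for the reload-start sound of the Figment pistol-caliber assault rifle.
# A single wave player plays one AutoRifle reload-start wave, with no randomization or modulation nodes.
# It uses weapon reload attenuation and the weapon foley sound class, with a raised volume multiplier and a concurrency override.
# The debug flag is left on.
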SoundCue /Figment_Assault_PistolCalibur_AR/Audio/SC_Assault_PistolCaliber_Reload_Start_Figment_Cue
root=WavePlayer_0
nodes [{id=WavePlayer_0, properties {soundWaveAssetPtr=/Game/Sounds/Fort_Weapons/AutoRifle/Assault_PistolCaliber/JW_AR_ReloadStart_01}}]
properties {
 attenuationSettings=&Fort_Weapon_Reload_Att
 bDebug=true
 bOverrideConcurrency=true
 maxDistance=2250
 soundClassObject=&Fort_WeaponFoley
 volumeMultiplier=1.25
}
